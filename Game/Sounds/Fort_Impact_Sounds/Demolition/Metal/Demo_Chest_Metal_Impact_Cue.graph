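# Sound cue for a metal chest demolition impact, routed to the impacts sound class with small-object hit attenuation.
# A mixer layers two random picks: a boosted set of big metal hits and a quieter set of utility pole impacts run through a modulator.
# An impacts concurrency set is assigned, alongside a local concurrency override set to prevent new sounds.
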SoundCue /Game/Sounds/Fort_Impact_Sounds/Demolition/Metal/Demo_Chest_Metal_Impact_Cue
root=Mixer_5
nodes [
 {id=Mixer_5, children [Random_8,Modulator_5], properties {inputVolume [2,0.5]}}
 {id=Modulator_5, children [Random_7]}
 {id=Random_7, children [WavePlayer_24, WavePlayer_25], properties {weights [1,1]}}
 {id=Random_8, children [WavePlayer_28, WavePlayer_30, WavePlayer_31], properties {weights [1,1,1]}}
 {id=WavePlayer_24, properties {soundWaveAssetPtr=/Game/Sounds/Fort_Impact_Sounds/Demolition/Utility_Pole/utility_pole_impact_01}}
 {id=WavePlayer_25, properties {soundWaveAssetPtr=/Game/Sounds/Fort_Impact_Sounds/Demolition/Utility_Pole/utility_pole_impact_02}}
 {id=WavePlayer_28, properties {soundWaveAssetPtr=&/Metal_Big_03}}
 {id=WavePlayer_30, properties {soundWaveAssetPtr=&/Metal_Big_04}}
 {id=WavePlayer_31, properties {soundWaveAssetPtr=&/Metal_Big_06}}
]
properties {
 attenuationSettings=/Game/Sounds/Attenuation/Fort_Demolition_Hit_Small_Sized_Object_Att
 maxDistance=4000
 soundClassObject=&Fort_Impacts
 volumeMultiplier=0.34
 concurrencyOverrides {resolutionRule=PreventNew}
 concurrencySet [&CONC_Object_Impacts_FourMax]
}
