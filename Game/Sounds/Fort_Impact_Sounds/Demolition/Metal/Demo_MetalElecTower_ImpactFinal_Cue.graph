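# Sound cue for the final impact of a demolished metal electrical tower.
# A quality-level switch picks between two layered mixes and a single pre-optimized wave, which serves both remaining levels.
# The mixes layer pitch-varied railing breaks, garage impacts and pole breaks delayed 0.3 seconds, with fewer random variants in the second.
# Medium-object destruction attenuation and a shared object-impact concurrency set govern playback.
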SoundCue /Game/Sounds/Fort_Impact_Sounds/Demolition/Metal/Demo_MetalElecTower_ImpactFinal_Cue
root=QualityLevel_0
nodes [
 {id=Delay_0, children [Random_13], properties {delayMax=0.3, delayMin=0.3}}
 {id=Delay_1, children [WavePlayer_7], properties {delayMax=0.3, delayMin=0.3}}
 {id=Mixer_0, children [Modulator_1,Modulator_3,Modulator_4], properties {inputVolume [1,1,1]}}
 {id=Mixer_5, children [Modulator_12,Random_10,Delay_0], properties {inputVolume [1,1,1]}}
 {id=Modulator_0, children [WavePlayer_0]}
 {id=Modulator_1, children [Random_0], properties {pitchMax=1.07, pitchMin=0.92}}
 {id=Modulator_12, children [Random_11], properties {pitchMax=1.07, pitchMin=0.92}}
 {id=Modulator_3, children [WavePlayer_5], properties {pitchMin=0.9}}
 {id=Modulator_4, children [Delay_1], properties {pitchMin=0.9}}
 {id=Modulator_5, children [WavePlayer_10]}
 {id=QualityLevel_0, children [Mixer_5,Mixer_0,Modulator_0,Modulator_5]}
 {id=Random_0, children [WavePlayer_1,WavePlayer_2], properties {weights [1,1]}}
 {id=Random_10, children [WavePlayer_32, WavePlayer_33], properties {weights [1,1]}}
 {id=Random_11, children [WavePlayer_45, WavePlayer_46, WavePlayer_47], properties {weights [1,1,1]}}
 {id=Random_13, children [WavePlayer_51, WavePlayer_44], properties {weights [1,1]}}
 {id=WavePlayer_0, properties {soundWaveAssetPtr=&/MetalElecTower_ImpactFinal_opt}}
 {id=WavePlayer_1, properties {soundWaveAssetPtr=&/Metal_Railing_Death02}}
 {id=WavePlayer_10, properties {soundWaveAssetPtr=&MetalElecTower_ImpactFinal_opt}}
 {id=WavePlayer_2, properties {soundWaveAssetPtr=&/Metal_Railing_Death04}}
 {id=WavePlayer_32, properties {soundWaveAssetPtr=&/Metal_Garage_Impact01}}
 {id=WavePlayer_33, properties {soundWaveAssetPtr=&/Metal_Garage_Impact02}}
 {id=WavePlayer_44, properties {soundWaveAssetPtr=&/Metal_Pole_Death06}}
 {id=WavePlayer_45, properties {soundWaveAssetPtr=&Metal_Railing_Death02}}
 {id=WavePlayer_46, properties {soundWaveAssetPtr=&Metal_Railing_Death04}}
 {id=WavePlayer_47, properties {soundWaveAssetPtr=&/Metal_Railing_Death03}}
 {id=WavePlayer_5, properties {soundWaveAssetPtr=&Metal_Garage_Impact02}}
 {id=WavePlayer_51, properties {soundWaveAssetPtr=&/Metal_Pole_Death03}}
 {id=WavePlayer_7, properties {soundWaveAssetPtr=&Metal_Pole_Death03}}
]
properties {
 attenuationSettings=/Game/Sounds/Attenuation/Fort_Demolition_Destroy_Medium_Sized_Object_Att
 maxDistance=4000
 soundClassObject=&Fort_Impacts_PSM
 volumeMultiplier=0.544
 concurrencyOverrides {resolutionRule=PreventNew}
 concurrencySet [&CONC_Object_Impacts_FourMax]
}
