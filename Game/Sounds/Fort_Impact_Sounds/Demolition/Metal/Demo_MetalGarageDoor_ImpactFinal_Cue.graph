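# Sound cue for the final impact when a metal garage door is demolished, using medium-sized object destruction attenuation.
# A quality-level node sends its first two levels to a random pick of three death waves and the other two to one optimized wave.
# A modulator at the root varies pitch slightly.
# Concurrency references a four-max object impact set, alongside an override that prevents new sounds.
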SoundCue /Game/Sounds/Fort_Impact_Sounds/Demolition/Metal/Demo_MetalGarageDoor_ImpactFinal_Cue
root=Modulator_2
nodes [
 {id=Modulator_2, children [QualityLevel_0], properties {pitchMax=1.07, pitchMin=0.92}}
 {id=QualityLevel_0, children [Random_15,Random_15,WavePlayer_3,WavePlayer_3]}
 {id=Random_15, children [WavePlayer_0,WavePlayer_2,WavePlayer_1], properties {weights [1,1,1]}}
 {id=WavePlayer_0, properties {soundWaveAssetPtr=&/Metal_Garage_Death01}}
 {id=WavePlayer_1, properties {soundWaveAssetPtr=&/Metal_Garage_Death04}}
 {id=WavePlayer_2, properties {soundWaveAssetPtr=&/Metal_Garage_Death02}}
 {id=WavePlayer_3, properties {soundWaveAssetPtr=&/Metal_Garage_Death04_opt}}
]
properties {
 attenuationSettings=/Game/Sounds/Attenuation/Fort_Demolition_Destroy_Medium_Sized_Object_Att
 maxDistance=4000
 soundClassObject=&Fort_Impacts_PSM
 volumeMultiplier=0.544
 concurrencyOverrides {resolutionRule=PreventNew}
 concurrencySet [&CONC_Object_Impacts_FourMax]
}
