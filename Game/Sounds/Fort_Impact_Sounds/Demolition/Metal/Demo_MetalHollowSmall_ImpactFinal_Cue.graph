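# Sound cue for the final demolition impact of a small hollow metal object, using medium-object destroy attenuation and a four-max impact concurrency set.
# A quality-level switch selects between two three-layer mixes and two cheaper randomizers of optimized single waves, with the second mix using fewer variations.
# Each mix layers randomized hollow tank, garage impact and small metal sounds through modulators, with the first two layers pitched up sharply.
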SoundCue /Game/Sounds/Fort_Impact_Sounds/Demolition/Metal/Demo_MetalHollowSmall_ImpactFinal_Cue
root=QualityLevel_0
nodes [
 {id=Mixer_0, children [Modulator_1,Modulator_2,Modulator_3], properties {inputVolume [0.6,0.8,1]}}
 {id=Mixer_2, children [Modulator_4,Modulator_5,Modulator_0], properties {inputVolume [0.6,0.8,1]}}
 {id=Modulator_0, children [Random_2]}
 {id=Modulator_1, children [Random_0], properties {pitchMax=2, pitchMin=1.95}}
 {id=Modulator_2, children [Random_1], properties {pitchMax=1.9, pitchMin=1.7}}
 {id=Modulator_3, children [Random_3]}
 {id=Modulator_4, children [Random_9], properties {pitchMax=2, pitchMin=1.95}}
 {id=Modulator_5, children [Random_11], properties {pitchMax=1.9, pitchMin=1.7}}
 {id=QualityLevel_0, children [Mixer_2,Mixer_0,Random_4,Random_5]}
 {id=Random_0, children [WavePlayer_1,WavePlayer_0], properties {weights [1,1]}}
 {id=Random_1, children [WavePlayer_2,WavePlayer_3], properties {weights [1,1]}}
 {id=Random_11, children [WavePlayer_40, WavePlayer_41, WavePlayer_42], properties {weights [1,1,1]}}
 {id=Random_2, children [WavePlayer_6,WavePlayer_7,WavePlayer_8], properties {weights [1,1,1]}}
 {id=Random_3, children [WavePlayer_4,WavePlayer_5], properties {weights [1,1]}}
 {id=Random_4, children [WavePlayer_9, WavePlayer_10], properties {weights [1,1]}}
 {id=Random_5, children [WavePlayer_11, WavePlayer_12], properties {weights [1,1]}}
 {id=Random_9, children [WavePlayer_36, WavePlayer_35, WavePlayer_34], properties {weights [1,1,1]}}
 {id=WavePlayer_0, properties {soundWaveAssetPtr=&/Metal_HollowTank_Death02}}
 {id=WavePlayer_1, properties {soundWaveAssetPtr=&/Metal_HollowTank_Death07}}
 {id=WavePlayer_10, properties {soundWaveAssetPtr=&/Metal_Hollow_Small_02_opt}}
 {id=WavePlayer_11, properties {soundWaveAssetPtr=&/Metal_Hollow_Small_01_opt}}
 {id=WavePlayer_12, properties {soundWaveAssetPtr=&Metal_Hollow_Small_02_opt}}
 {id=WavePlayer_2, properties {soundWaveAssetPtr=&/Metal_Garage_Impact04}}
 {id=WavePlayer_3, properties {soundWaveAssetPtr=&/Metal_Garage_Impact05}}
 {id=WavePlayer_34, properties {soundWaveAssetPtr=&/Metal_HollowTank_Death05}}
 {id=WavePlayer_35, properties {soundWaveAssetPtr=&Metal_HollowTank_Death02}}
 {id=WavePlayer_36, properties {soundWaveAssetPtr=&Metal_HollowTank_Death07}}
 {id=WavePlayer_4, properties {soundWaveAssetPtr=&/Metal_Small_Death04}}
 {id=WavePlayer_40, properties {soundWaveAssetPtr=&Metal_Garage_Impact04}}
 {id=WavePlayer_41, properties {soundWaveAssetPtr=&Metal_Garage_Impact05}}
 {id=WavePlayer_42, properties {soundWaveAssetPtr=&/Metal_Garage_Impact06}}
 {id=WavePlayer_5, properties {soundWaveAssetPtr=&/Metal_Small_Death02}}
 {id=WavePlayer_6, properties {soundWaveAssetPtr=&Metal_Small_Death04}}
 {id=WavePlayer_7, properties {soundWaveAssetPtr=&Metal_Small_Death02}}
 {id=WavePlayer_8, properties {soundWaveAssetPtr=&/Metal_Small_Death03}}
 {id=WavePlayer_9, properties {soundWaveAssetPtr=&Metal_Hollow_Small_01_opt}}
]
properties {
 attenuationSettings=/Game/Sounds/Attenuation/Fort_Demolition_Destroy_Medium_Sized_Object_Att
 maxDistance=4000
 pitchMultiplier=1.1
 soundClassObject=&Fort_Impacts_PSM
 volumeMultiplier=0.544
 concurrencyOverrides {resolutionRule=PreventNew}
 concurrencySet [&CONC_Object_Impacts_FourMax]
}
