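# Sound cue for small hollow metal demolition impacts, built from hollow tank impact recordings played at a doubled pitch multiplier and reduced volume.
# A quality-level root selects one of four branches, thinning from three random waves to two, two, then a single wave, each with slight pitch modulation.
# It uses small-object demolition attenuation, the impacts sound class and a shared object-impact concurrency set to limit overlap.
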SoundCue /Game/Sounds/Fort_Impact_Sounds/Demolition/Metal/Demo_MetalHollowSmall_Impact_Cue
root=QualityLevel_0
nodes [
 {id=Modulator_0, children [Random_1], properties {pitchMax=1.08, pitchMin=0.92}}
 {id=Modulator_3, children [Random_2], properties {pitchMax=1.08, pitchMin=0.92}}
 {id=Modulator_4, children [Random_3], properties {pitchMax=1.08, pitchMin=0.92}}
 {id=Modulator_5, children [WavePlayer_8], properties {pitchMax=1.08, pitchMin=0.92}}
 {id=QualityLevel_0, children [Modulator_0,Modulator_3,Modulator_4,Modulator_5]}
 {id=Random_1, children [WavePlayer_0,WavePlayer_1,WavePlayer_2], properties {weights [1,1,1]}}
 {id=Random_2, children [WavePlayer_3,WavePlayer_4], properties {weights [1,1]}}
 {id=Random_3, children [WavePlayer_6,WavePlayer_7], properties {weights [1,1]}}
 {id=WavePlayer_0, properties {soundWaveAssetPtr=&/Metal_HollowTank_Impact_mix_01}}
 {id=WavePlayer_1, properties {soundWaveAssetPtr=&/Metal_HollowTank_Impact_mix_02}}
 {id=WavePlayer_2, properties {soundWaveAssetPtr=&/Metal_HollowTank_Impact_mix_03}}
 {id=WavePlayer_3, properties {soundWaveAssetPtr=&Metal_HollowTank_Impact_mix_01}}
 {id=WavePlayer_4, properties {soundWaveAssetPtr=&Metal_HollowTank_Impact_mix_02}}
 {id=WavePlayer_6, properties {soundWaveAssetPtr=&Metal_HollowTank_Impact_mix_01}}
 {id=WavePlayer_7, properties {soundWaveAssetPtr=&Metal_HollowTank_Impact_mix_02}}
 {id=WavePlayer_8, properties {soundWaveAssetPtr=&Metal_HollowTank_Impact_mix_01}}
]
properties {
 attenuationSettings=/Game/Sounds/Attenuation/Fort_Demolition_Hit_Small_Sized_Object_Att
 maxDistance=4000
 pitchMultiplier=2
 soundClassObject=&Fort_Impacts
 volumeMultiplier=0.34
 concurrencyOverrides {maxCount=3, resolutionRule=StopOldest}
 concurrencySet [&CONC_Object_Impacts_FourMax]
}
