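# Sound cue for the final demolition impact that destroys a hollow metal tank object.
# A quality-level node selects one of four branches, whose variety steps down from a random pick of three death mixes, to two, to a single wave in the last two.
# Each branch adds slight pitch modulation.
# Playback uses medium-object destruction attenuation and a shared object-impact concurrency group.
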SoundCue /Game/Sounds/Fort_Impact_Sounds/Demolition/Metal/Demo_MetalHollowTank_ImpactFinal_Cue
root=QualityLevel_1
nodes [
 {id=Modulator_0, children [Random_0], properties {pitchMax=1.07, pitchMin=0.92}}
 {id=Modulator_1, children [Random_1], properties {pitchMax=1.07, pitchMin=0.92}}
 {id=Modulator_2, children [WavePlayer_6], properties {pitchMax=1.07, pitchMin=0.92}}
 {id=Modulator_3, children [WavePlayer_8], properties {pitchMax=1.07, pitchMin=0.92}}
 {id=QualityLevel_1, children [Modulator_0,Modulator_1,Modulator_2,Modulator_3]}
 {id=Random_0, children [WavePlayer_0,WavePlayer_1,WavePlayer_2], properties {weights [1,1,1]}}
 {id=Random_1, children [WavePlayer_3,WavePlayer_4], properties {weights [1,1]}}
 {id=WavePlayer_0, properties {soundWaveAssetPtr=&/Metal_HollowTank_Death_mix_02}}
 {id=WavePlayer_1, properties {soundWaveAssetPtr=&/Metal_HollowTank_Death_mix_05}}
 {id=WavePlayer_2, properties {soundWaveAssetPtr=&/Metal_HollowTank_Death_mix_07}}
 {id=WavePlayer_3, properties {soundWaveAssetPtr=&Metal_HollowTank_Death_mix_02}}
 {id=WavePlayer_4, properties {soundWaveAssetPtr=&Metal_HollowTank_Death_mix_05}}
 {id=WavePlayer_6, properties {soundWaveAssetPtr=&Metal_HollowTank_Death_mix_02}}
 {id=WavePlayer_8, properties {soundWaveAssetPtr=&Metal_HollowTank_Death_mix_02}}
]
properties {
 attenuationSettings=/Game/Sounds/Attenuation/Fort_Demolition_Destroy_Medium_Sized_Object_Att
 maxDistance=4000
 pitchMultiplier=1.1
 soundClassObject=&Fort_Impacts_PSM
 volumeMultiplier=0.544
 concurrencyOverrides {resolutionRule=PreventNew}
 concurrencySet [&CONC_Object_Impacts_FourMax]
}
