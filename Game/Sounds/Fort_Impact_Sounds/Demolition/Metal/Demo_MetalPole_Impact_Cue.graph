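# Sound cue for metal pole impacts in the demolition sound set, routed to the impacts sound class with small-object hit attenuation.
# A root quality-level node selects among four branches: a three-wave random pick, a two-wave random pick, and two single-wave players.
# Each branch passes through a modulator for slight pitch variation.
# Concurrency settings limit overlapping impacts, with an override that stops the oldest.
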SoundCue /Game/Sounds/Fort_Impact_Sounds/Demolition/Metal/Demo_MetalPole_Impact_Cue
root=QualityLevel_0
nodes [
 {id=Modulator_0, children [Random_0], properties {pitchMax=1.07, pitchMin=0.92}}
 {id=Modulator_1, children [WavePlayer_3], properties {pitchMax=1.07, pitchMin=0.92}}
 {id=Modulator_2, children [Random_4], properties {pitchMax=1.07, pitchMin=0.92}}
 {id=Modulator_3, children [WavePlayer_4], properties {pitchMax=1.07, pitchMin=0.92}}
 {id=QualityLevel_0, children [Modulator_2,Modulator_0,Modulator_1,Modulator_3]}
 {id=Random_0, children [WavePlayer_0,WavePlayer_1], properties {weights [1,1]}}
 {id=Random_4, children [WavePlayer_36, WavePlayer_20, WavePlayer_37], properties {weights [1,1,1]}}
 {id=WavePlayer_0, properties {soundWaveAssetPtr=&/Metal_Pole_Impact01}}
 {id=WavePlayer_1, properties {soundWaveAssetPtr=&/Metal_Pole_Impact05}}
 {id=WavePlayer_20, properties {soundWaveAssetPtr=&Metal_Pole_Impact05}}
 {id=WavePlayer_3, properties {soundWaveAssetPtr=&Metal_Pole_Impact01}}
 {id=WavePlayer_36, properties {soundWaveAssetPtr=&Metal_Pole_Impact01}}
 {id=WavePlayer_37, properties {soundWaveAssetPtr=&/Metal_Pole_Impact06}}
 {id=WavePlayer_4, properties {soundWaveAssetPtr=&Metal_Pole_Impact01}}
]
properties {
 attenuationSettings=/Game/Sounds/Attenuation/Fort_Demolition_Hit_Small_Sized_Object_Att
 maxDistance=4000
 soundClassObject=&Fort_Impacts
 volumeMultiplier=0.34
 concurrencyOverrides {maxCount=3, resolutionRule=StopOldest}
 concurrencySet [&CONC_Object_Impacts_FourMax]
}
