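# Sound cue for the final impact sound of a demolished metal sign, using medium-sized destruction attenuation.
# A quality-level switch sends its first two slots to an equal-weight random pick of three impact waves, and its last two to a single wave.
# Both branches pass through a modulator for slight pitch variation.
# Playback uses the impacts sound class and a four-voice object-impact concurrency limit.
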SoundCue /Game/Sounds/Fort_Impact_Sounds/Demolition/Metal/Demo_MetalSign_ImpactFinal_Cue
root=QualityLevel_0
nodes [
 {id=Modulator_0, children [WavePlayer_0], properties {pitchMax=1.07, pitchMin=0.92}}
 {id=Modulator_2, children [Random_6], properties {pitchMax=1.07, pitchMin=0.92}}
 {id=QualityLevel_0, children [Modulator_2,Modulator_2,Modulator_0,Modulator_0]}
 {id=Random_6, children [WavePlayer_22, WavePlayer_23, WavePlayer_24], properties {weights [1,1,1]}}
 {id=WavePlayer_0, properties {soundWaveAssetPtr=&/Metal_Sign_ImpactFinal02}}
 {id=WavePlayer_22, properties {soundWaveAssetPtr=&/Metal_Sign_ImpactFinal01}}
 {id=WavePlayer_23, properties {soundWaveAssetPtr=&Metal_Sign_ImpactFinal02}}
 {id=WavePlayer_24, properties {soundWaveAssetPtr=&/Metal_Sign_ImpactFinal03}}
]
properties {
 attenuationSettings=/Game/Sounds/Attenuation/Fort_Demolition_Destroy_Medium_Sized_Object_Att
 maxDistance=4000
 soundClassObject=&Fort_Impacts_PSM
 volumeMultiplier=0.544
 concurrencyOverrides {maxCount=4, resolutionRule=StopOldest}
 concurrencySet [&CONC_Object_Impacts_FourMax]
}
