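# Sound cue for the final impact when a small metal object is demolished, using medium-object destruction attenuation and a shared object-impact concurrency set.
# A quality-level root selects among four branches of differing cost.
# The first two mix randomized small-metal death sounds with half-volume, pitched-up metal pole hits, drawing on three and two variants per layer respectively.
# The last two each play one optimized wave with slight pitch variation.
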SoundCue /Game/Sounds/Fort_Impact_Sounds/Demolition/Metal/Demo_MetalSmall_ImpactFinal_Cue
root=QualityLevel_0
nodes [
 {id=Mixer_0, children [Modulator_0,Modulator_1], properties {inputVolume [1,0.5]}}
 {id=Mixer_4, children [Modulator_2,Modulator_13], properties {inputVolume [1,0.5]}}
 {id=Modulator_0, children [Random_0], properties {pitchMax=1.07, pitchMin=0.92}}
 {id=Modulator_1, children [Random_1], properties {pitchMax=1.62, pitchMin=1.5}}
 {id=Modulator_13, children [Random_18], properties {pitchMax=1.62, pitchMin=1.5}}
 {id=Modulator_2, children [Random_7], properties {pitchMax=1.07, pitchMin=0.92}}
 {id=Modulator_3, children [WavePlayer_0], properties {pitchMax=1.07, pitchMin=0.92}}
 {id=Modulator_4, children [WavePlayer_5], properties {pitchMax=1.07, pitchMin=0.92}}
 {id=QualityLevel_0, children [Mixer_4,Mixer_0,Modulator_3,Modulator_4]}
 {id=Random_0, children [WavePlayer_1,WavePlayer_2], properties {weights [1,1]}}
 {id=Random_1, children [WavePlayer_4,WavePlayer_3], properties {weights [1,1]}}
 {id=Random_18, children [WavePlayer_64, WavePlayer_62, WavePlayer_63], properties {weights [1,1,1]}}
 {id=Random_7, children [WavePlayer_28, WavePlayer_29, WavePlayer_71], properties {weights [1,1,1]}}
 {id=WavePlayer_0, properties {soundWaveAssetPtr=&/Metal_Small_ImpactFinal_opt}}
 {id=WavePlayer_1, properties {soundWaveAssetPtr=&/Metal_Small_Death04}}
 {id=WavePlayer_2, properties {soundWaveAssetPtr=&/Metal_Small_Death02}}
 {id=WavePlayer_28, properties {soundWaveAssetPtr=&Metal_Small_Death04}}
 {id=WavePlayer_29, properties {soundWaveAssetPtr=&Metal_Small_Death02}}
 {id=WavePlayer_3, properties {soundWaveAssetPtr=&/Metal_Pole_Death06}}
 {id=WavePlayer_4, properties {soundWaveAssetPtr=&/Metal_Pole_Death03}}
 {id=WavePlayer_5, properties {soundWaveAssetPtr=&Metal_Small_ImpactFinal_opt}}
 {id=WavePlayer_62, properties {soundWaveAssetPtr=&Metal_Pole_Death06}}
 {id=WavePlayer_63, properties {soundWaveAssetPtr=&/Metal_Pole_Death02}}
 {id=WavePlayer_64, properties {soundWaveAssetPtr=&Metal_Pole_Death03}}
 {id=WavePlayer_71, properties {soundWaveAssetPtr=&/Metal_Small_Death03}}
]
properties {
 attenuationSettings=/Game/Sounds/Attenuation/Fort_Demolition_Destroy_Medium_Sized_Object_Att
 maxDistance=4000
 pitchMultiplier=0.9
 soundClassObject=&Fort_Impacts_PSM
 volumeMultiplier=0.34
 concurrencyOverrides {resolutionRule=PreventNew}
 concurrencySet [&CONC_Object_Impacts_FourMax]
}
