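# Sound cue for the final demolition impact of a metal gym barbell, layered from gym barbell and explosion impact samples.
# A mixer plays three random picks at once: a barbell impact, a generic explosion impact layer, and a barbell final impact offset by a short random delay.
# A modulator varies pitch per play, with medium-object destroy attenuation and an object-impact concurrency set limiting stacking.
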SoundCue /Game/Sounds/Fort_Impact_Sounds/Demolition/Metal/Demo_Metal_Barbell_ImpactFinal_Cue
root=Modulator_29
nodes [
 {id=Delay_9, children [Random_50], properties {delayMax=0.25, delayMin=0.15}}
 {id=Mixer_10, children [Random_51,Delay_9,Random_49], properties {inputVolume [1,1,1]}}
 {id=Modulator_29, children [Mixer_10], properties {pitchMax=1.07, pitchMin=0.92}}
 {id=Random_49, children [WavePlayer_159, WavePlayer_160], properties {weights [1,1]}}
 {id=Random_50, children [WavePlayer_161, WavePlayer_162, WavePlayer_163], properties {weights [1,1,1]}}
 {id=Random_51, children [WavePlayer_164, WavePlayer_165, WavePlayer_166], properties {weights [1,1,1]}}
 {id=WavePlayer_159, properties {soundWaveAssetPtr=/Game/Sounds/Fort_Impact_Sounds/Demolition/Defaults/Explo_Impact_Layer01}}
 {id=WavePlayer_160, properties {soundWaveAssetPtr=/Game/Sounds/Fort_Impact_Sounds/Demolition/Defaults/Explo_Impact_Layer02}}
 {id=WavePlayer_161, properties {soundWaveAssetPtr=/Game/Sounds/Fort_Impact_Sounds/Demolition/Gym/Gym_Barbell_ImpactFinal04}}
 {id=WavePlayer_162, properties {soundWaveAssetPtr=/Game/Sounds/Fort_Impact_Sounds/Demolition/Gym/Gym_Barbell_ImpactFinal02}}
 {id=WavePlayer_163, properties {soundWaveAssetPtr=/Game/Sounds/Fort_Impact_Sounds/Demolition/Gym/Gym_Barbell_ImpactFinal01}}
 {id=WavePlayer_164, properties {soundWaveAssetPtr=/Game/Sounds/Fort_Impact_Sounds/Demolition/Gym/Gym_Barbell_Impact04}}
 {id=WavePlayer_165, properties {soundWaveAssetPtr=/Game/Sounds/Fort_Impact_Sounds/Demolition/Gym/Gym_Barbell_Impact01}}
 {id=WavePlayer_166, properties {soundWaveAssetPtr=/Game/Sounds/Fort_Impact_Sounds/Demolition/Gym/Gym_Barbell_Impact02}}
]
properties {
 attenuationSettings=/Game/Sounds/Attenuation/Fort_Demolition_Destroy_Medium_Sized_Object_Att
 maxDistance=4000
 soundClassObject=&Fort_Impacts_PSM
 volumeMultiplier=0.544
 concurrencyOverrides {maxCount=3, resolutionRule=StopOldest}
 concurrencySet [&CONC_Object_Impacts_FourMax]
}
