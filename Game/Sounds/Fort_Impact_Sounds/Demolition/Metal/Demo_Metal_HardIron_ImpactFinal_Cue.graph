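# Sound cue for the final demolition impact on hard iron metal objects, using the Metal_Cannon_ImpactFinal wave set.
# A Random node picks one of five equally weighted impact waves, and a Modulator applies slight pitch variation.
# It plays at reduced volume through the impacts sound class, with medium-sized-object destruction attenuation and a concurrency set for object impacts.
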SoundCue /Game/Sounds/Fort_Impact_Sounds/Demolition/Metal/Demo_Metal_HardIron_ImpactFinal_Cue
root=Modulator_2
nodes [
 {id=Modulator_2, children [Random_7], properties {pitchMax=1.07, pitchMin=0.92}}
 {id=Random_7, children [WavePlayer_35, WavePlayer_36, WavePlayer_32, WavePlayer_33, WavePlayer_34], properties {weights [1,1,1,1,1]}}
 {id=WavePlayer_32, properties {soundWaveAssetPtr=&/Metal_Cannon_ImpactFinal03}}
 {id=WavePlayer_33, properties {soundWaveAssetPtr=&/Metal_Cannon_ImpactFinal04}}
 {id=WavePlayer_34, properties {soundWaveAssetPtr=&/Metal_Cannon_ImpactFinal05}}
 {id=WavePlayer_35, properties {soundWaveAssetPtr=&/Metal_Cannon_ImpactFinal01}}
 {id=WavePlayer_36, properties {soundWaveAssetPtr=&/Metal_Cannon_ImpactFinal02}}
]
properties {
 attenuationSettings=/Game/Sounds/Attenuation/Fort_Demolition_Destroy_Medium_Sized_Object_Att
 maxDistance=4000
 soundClassObject=&Fort_Impacts_PSM
 volumeMultiplier=0.544
 concurrencySet [&CONC_Object_Impacts_FourMax]
}
