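# Sound cue for metal machine demolition impacts, played as a small-object hit in the impacts sound class.
# A quality-level switch at the root selects one of four branches: a random pick of four waves, a random pick of three, or one of two identical single-wave branches.
# Each branch has a modulator adding slight pitch variation.
# Shared attenuation and an object-impact concurrency set govern distance falloff and voice limiting.
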SoundCue /Game/Sounds/Fort_Impact_Sounds/Demolition/Metal/Demo_Metal_Machine_Impact_Cue
root=QualityLevel_0
nodes [
 {id=Modulator_0, children [Random_0], properties {pitchMax=1.07, pitchMin=0.92}}
 {id=Modulator_1, children [Random_17], properties {pitchMax=1.07, pitchMin=0.92}}
 {id=Modulator_2, children [WavePlayer_4], properties {pitchMax=1.07, pitchMin=0.92}}
 {id=Modulator_3, children [WavePlayer_5], properties {pitchMax=1.07, pitchMin=0.92}}
 {id=QualityLevel_0, children [Modulator_1,Modulator_0,Modulator_2,Modulator_3]}
 {id=Random_0, children [WavePlayer_0,WavePlayer_1,WavePlayer_2], properties {weights [1,1,1]}}
 {id=Random_17, children [WavePlayer_62, WavePlayer_63, WavePlayer_64, WavePlayer_65], properties {weights [1,1,1,1]}}
 {id=WavePlayer_0, properties {soundWaveAssetPtr=&/Metal_Machine_Impact02}}
 {id=WavePlayer_1, properties {soundWaveAssetPtr=&/Metal_Machine_Impact03}}
 {id=WavePlayer_2, properties {soundWaveAssetPtr=&/Metal_Machine_Impact04}}
 {id=WavePlayer_4, properties {soundWaveAssetPtr=&Metal_Machine_Impact03}}
 {id=WavePlayer_5, properties {soundWaveAssetPtr=&Metal_Machine_Impact03}}
 {id=WavePlayer_62, properties {soundWaveAssetPtr=&Metal_Machine_Impact02}}
 {id=WavePlayer_63, properties {soundWaveAssetPtr=&Metal_Machine_Impact03}}
 {id=WavePlayer_64, properties {soundWaveAssetPtr=&Metal_Machine_Impact04}}
 {id=WavePlayer_65, properties {soundWaveAssetPtr=&/Metal_Machine_Impact05}}
]
properties {
 attenuationSettings=/Game/Sounds/Attenuation/Fort_Demolition_Hit_Small_Sized_Object_Att
 maxDistance=4000
 soundClassObject=&Fort_Impacts
 volumeMultiplier=0.48
 concurrencyOverrides {maxCount=3, resolutionRule=StopOldest}
 concurrencySet [&CONC_Object_Impacts_FourMax]
}
